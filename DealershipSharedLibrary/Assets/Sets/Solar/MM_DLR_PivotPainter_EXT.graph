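# Translucent, unlit, two-sided Pivot Painter material for Niagara mesh particles, moving and dissolving a mesh's individual elements.
# Per-element pivots, axes and random values feed rotation, expansion, scale and curl-noise offsets, driven by Niagara dynamic parameters.
# A static switch selects between summing those offsets and using a Sprite function with overridden pivot.
# Particle color and sharpened Fresnel build emissive; Fresnel, particle alpha and Perlin dissolve build opacity, with optional sphere mask.
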Material /DealershipSharedLibrary/Assets/Sets/Solar/MM_DLR_PivotPainter_EXT
graph {
 properties {
  bAutomaticallySetUsageInEditor=false
  bHasPixelAnimation=true
  bUsedWithNiagaraMeshParticles=true
  blendMode=BLEND_Translucent
  previewMesh=/DealershipSharedLibrary/Assets/Sets/Quasar/SM_PivotPainter_LineMesh
  shadingModel=MSM_Unlit
  twoSided=true
  usedShadingModels=Unlit
 }
 nodes [
  Constant3Vector() {constant=<rgba 100 0 0 0>}
  Constant()
  Constant() {r=20}
  DynamicParameter() {parameterIndex=3, paramNames [Dissolve, DisoslveContrast, -, -]}
  n1 = TextureCoordinate() {coordinateIndex=1}
  n2 = TextureCoordinate() {coordinateIndex=3}
  n3 = Add(A=OneMinus(None=ComponentMask(None=@n1) {g=true}), B=Multiply(A=ComponentMask(None=@n2) {g=true}, B=Constant() {r=1}))
  n4 = ScalarParameter() {defaultValue=1, parameterName=BrightColorAlpha}
  n5 = OneMinus(None=Subtract(A=Subtract(A=@n3, B=@n4), B=LinearInterpolate() {constA=-1}))
  LinearInterpolate(A=VertexColor(), B=Constant(), Alpha=@n5)
  n6 = Fresnel(Normal=NamedRerouteUsage()) {baseReflectFraction=0.4, exponent=1}
  n7 = TextureSample() {samplerType=SAMPLERTYPE_LinearColor, texture=/Game/Effects/Fort_Effects/Textures/Tile/T_PerlinRGB_Displacement}
  n8 = DynamicParameter() {parameterIndex=3, paramNames [Dissolve, DisoslveContrast, -, -]}
  n9 = OneMinus(None=@n8.Dissolve)
  n10 = Saturate(None=Subtract(A=Add(A=@n7.R), B=Multiply(A=@n9) {constB=2}))
  n11 = Multiply(
   A=@n6
   B=Multiply(
    A=ParticleColor().A
    B=MaterialFunctionCall(In=@n10, Contrast=@n8.DisoslveContrast) {materialFunction=&CheapContrast}.Result
   )
  )
  n12 = TextureCoordinate()
  n13 = Constant2Vector() {g=0.5, r=0.5}
  n14 = Constant() {r=0.5}
  n15 = Constant() {r=1}
  n16 = StaticSwitchParameter(True=Multiply(A=@n11, B=SphereMask(A=@n12, B=@n13, Radius=@n14, Hardness=@n15)), False=@n11) {parameterName=Use SphereMask on/off}
  n17 = Subtract(A=NamedRerouteUsage(), B=ParticlePositionWS())
  n18 = NamedRerouteUsage()
  n19 = OneMinus(None=ComponentMask(None=TextureCoordinate() {coordinateIndex=1}) {g=true})
  n20 = Saturate(None=NamedRerouteUsage())
  n21 = LinearInterpolate(Alpha=NamedRerouteUsage()) {constA=-1}
  n22 = Multiply(A=Subtract(A=Subtract(A=@n19, B=@n20), B=@n21), B=Constant() {r=10})
  n23 = ComponentMask(None=TextureCoordinate() {coordinateIndex=3}) {g=true}
  n24 = NamedRerouteUsage()
  n25 = Subtract(A=Add(A=Multiply(A=@n23, B=@n24), B=@n19), B=@n21)
  n26 = Multiply(A=Saturate(None=OneMinus(None=@n22)), B=Multiply(A=Saturate(None=@n25), B=VertexTangentWS()))
  n27 = Constant() {r=-1}
  n28 = Time()
  n29 = NamedRerouteUsage()
  n30 = TextureCoordinate() {coordinateIndex=3}
  n31 = Add(A=Add(A=Multiply(A=@n28, B=@n29)) {constB=0.5}, B=Multiply(A=ComponentMask(None=@n30) {g=true}, B=NamedRerouteUsage()) {constB=0.05})
  n32 = VertexColor()
  n33 = Sine(None=Add(A=@n31, B=@n32.R)) {period=0.75}
  n34 = Constant3Vector() {constant=<rgba 1 0 0 0>}
  n35 = Cosine(None=Add(A=@n31, B=@n32.G))
  n36 = Constant3Vector() {constant=<rgba 0 1 0 0>}
  n37 = Sine(None=Add(A=@n31, B=@n32.B)) {period=1.2}
  n38 = Add(A=Add(A=Multiply(A=@n33, B=@n34), B=Multiply(A=@n35, B=@n36)), B=Multiply(A=@n37, B=Constant3Vector() {constant=<rgba 0 0 1 0>}))
  n39 = Multiply(A=@n33, B=@n35)
  n40 = Multiply(
   A=LinearInterpolate(A=Multiply(A=@n27, B=@n38), B=@n38, Alpha=Multiply(A=@n39, B=@n37))
   B=Multiply(A=Normalize(VectorInput=@n17), B=Multiply(A=@n18, B=@n26))
  )
  n41 = Add(A=NamedRerouteUsage(), B=@n40)
  n42 = Multiply(A=NamedRerouteUsage(), B=NamedRerouteUsage())
  n43 = OneMinus(None=Multiply(A=NamedRerouteUsage(), B=OneMinus(None=@n42)))
  n44 = StaticBool() {value=true}
  n45 = Add(A=Add(A=NamedRerouteUsage(), B=NamedRerouteUsage()), B=NamedRerouteUsage())
  n46 = NamedRerouteUsage()
  LinearInterpolate(
   A=Multiply(A=@n16, B=@n40)
   B=@n16
   Alpha=StaticSwitchParameter(
    True=MaterialFunctionCall!(Center Location=@n41, 0-1 UVs, XY scale=@n43, OverridePivot=@n44, "Normals (Optional)") {materialFunction=&Sprite}.World position offset
    False=Add(A=@n45, B=@n46)
   ) {parameterName=Use Sprite Pivot Pointer}
  )
  MaterialFunctionCall(Alpha Threshold=@n11) {materialFunction=&DitherTemporalAA_Fortnite}
  Multiply(A=VertexNormalWS())
  n47 = Add(A=NamedRerouteUsage(), B=ParticleRandom())
  NamedRerouteDeclaration(None=ConstantBiasScale(None=Sine(None=@n47))) {name=Final Random, nodeColor=<rgba 1 0.0156 0.6498 1>}
  n48 = MaterialFunctionCall() {materialFunction=/Engine/Functions/Engine_MaterialFunctions02/PivotPainter/PivotPainter_PerObjectData}
  NamedRerouteDeclaration(None=@n48.Pivot Position) {name=PivotPosition, nodeColor=<rgba 1 0.0156 0.6498 1>}
  n49 = Normalize(VectorInput=NamedRerouteUsage())
  n50 = Multiply(A=OneMinus(None=NamedRerouteUsage()), B=NamedRerouteUsage())
  n51 = Multiply(A=NamedRerouteUsage(), B=Add(A=Saturate(None=@n50)))
  NamedRerouteDeclaration(None=RotateAboutAxis(NormalizedRotationAxis=@n49, RotationAngle=@n51, PivotPoint=NamedRerouteUsage(), Position=WorldPosition().XYZ)) {name=WPO Rotation, nodeColor=<rgba 0.1192 0.1512 1 1>}
  n52 = Normalize(VectorInput=Subtract(A=NamedRerouteUsage(), B=ParticlePositionWS()))
  n53 = NamedRerouteUsage()
  n54 = NamedRerouteUsage()
  n55 = Multiply(A=NamedRerouteUsage(), B=OneMinus(None=Multiply(A=@n53, B=@n54)))
  NamedRerouteDeclaration(None=Multiply(A=Multiply(A=@n52, B=@n55), B=NamedRerouteUsage())) {name=WPO Expansion, nodeColor=<rgba 0.1192 0.1512 1 1>}
  n56 = NamedRerouteUsage()
  n57 = WorldPosition()
  n58 = NamedRerouteUsage()
  n59 = NamedRerouteUsage()
  n60 = Multiply(A=NamedRerouteUsage(), B=OneMinus(None=Multiply(A=@n58, B=@n59)))
  NamedRerouteDeclaration(None=Multiply(A=Subtract(A=@n56, B=@n57.XYZ), B=OneMinus(None=@n60))) {name=WPO Scale, nodeColor=<rgba 0.1192 0.1512 1 1>}
  n61 = DynamicParameter() {paramNames [Expansion, Expansion Range, Scale, Scale Random]}
  NamedRerouteDeclaration(None=@n61.Scale) {name=DP Scale, nodeColor=<rgba 0.4459 0.0012 1 1>}
  NamedRerouteDeclaration(None=@n61.Scale Random) {name=DP Scale Random, nodeColor=<rgba 0.4459 0.0012 1 1>}
  n62 = DynamicParameter() {parameterIndex=2, paramNames [Noise Amplitude, Noise Frequency, Noise Time, ScaleFactor]}
  NamedRerouteDeclaration(None=@n62.Noise Amplitude) {name=DP Noise Amplitude, nodeColor=<rgba 0.4459 0.0012 1 1>}
  NamedRerouteDeclaration(None=@n48.Random Value Per Element) {name=Random Value Per Element, nodeColor=<rgba 1 0.0156 0.6498 1>}
  NamedRerouteDeclaration(None=@n62.Noise Time) {name=DP Noise Time, nodeColor=<rgba 0.4459 0.0012 1 1>}
  NamedRerouteDeclaration(None=@n62.ScaleFactor) {name=DP Scale Factor, nodeColor=<rgba 0.4459 0.0012 1 1>}
  n63 = DynamicParameter() {parameterIndex=1, paramNames [Rotation, Rotation Random, Sprite Transition, Sprite TransitionRandomAmount]}
  NamedRerouteDeclaration(None=@n63.Sprite Transition) {name=DP_Sprite_Transition, nodeColor=<rgba 0.4459 0.0012 1 1>}
  n64 = Normalize(VectorInput=ConstantBiasScale(None=VertexColor()) {bias=-0.5, scale=2})
  NamedRerouteDeclaration(None=Normalize(VectorInput=Transform(None=@n64) {transformSourceType=TRANSFORMSOURCE_Instance})) {name=X-Axis Vector, nodeColor=<rgba 1 0.0156 0.6498 1>}
  NamedRerouteDeclaration(None=@n63.Sprite TransitionRandomAmount) {name=DP_Sprite_TransitionRandomAmount, nodeColor=<rgba 0.4459 0.0012 1 1>}
  NamedRerouteDeclaration() {name=DP Alpha Transition, nodeColor=<rgba 0.4459 0.0012 1 1>}
  n65 = OneMinus(None=ScalarParameter() {parameterName=Vertex push})
  n66 = VertexNormalWS()
  n67 = ScalarParameter() {defaultValue=1}
  n68 = Subtract(A=NamedRerouteUsage(), B=ParticlePositionWS())
  n69 = Multiply(A=Time(), B=ScalarParameter() {defaultValue=0.5, parameterName=Time})
  n70 = Normalize(VectorInput=OneMinus(None=Add(A=@n68, B=@n69)))
  NamedRerouteDeclaration(None=Multiply(A=Add(A=@n65, B=@n66), B=Multiply(A=@n67, B=@n70))) {name=Outwards WPO, nodeColor=<rgba 1 0.4688 0 1>}
  NamedRerouteDeclaration() {name=DP Alpha Transition Range, nodeColor=<rgba 0.4459 0.0012 1 1>}
  NamedRerouteDeclaration(None=@n63.Rotation) {name=DP Rotation, nodeColor=<rgba 0.4459 0.0012 1 1>}
  NamedRerouteDeclaration(None=@n63.Rotation Random) {name=DP Rotation Random, nodeColor=<rgba 0.4459 0.0012 1 1>}
  NamedRerouteDeclaration(None=@n61.Expansion) {name=DP Expansion, nodeColor=<rgba 0.4459 0.0012 1 1>}
  NamedRerouteDeclaration(None=@n61.Expansion Range) {name=DP Expansion Range, nodeColor=<rgba 0.4459 0.0012 1 1>}
  n71 = NamedRerouteUsage()
  n72 = ParticlePositionWS()
  n73 = Divide(A=NamedRerouteUsage(), B=NamedRerouteUsage())
  n74 = NamedRerouteUsage()
  n75 = ParticleRandom()
  n76 = Add(
   A=Multiply(A=Subtract(A=@n71, B=@n72), B=Multiply(A=@n73) {constB=0.001}) {constB=0.1}
   B=Multiply(A=VectorParameter() {parameterName=Noise Speed, defaultValue=<rgba 1 0 1 1>}.RGB, B=Add(A=@n74, B=@n75))
  )
  n77 = NamedRerouteUsage()
  n78 = NamedRerouteUsage()
  NamedRerouteDeclaration(None=Multiply(
    A=TextureSample(UVs=@n76) {samplerType=SAMPLERTYPE_LinearColor, texture=/Volumetrics/Content/VolumeTextures/Textures/VT_CurlNoise}.RGB
    B=Multiply(A=@n77, B=@n78)
   )) {name=WPO Noise, nodeColor=<rgba 0.1192 0.1512 1 1>}
  NamedRerouteDeclaration(None=@n62.Noise Frequency) {name=DP Noise Frequency, nodeColor=<rgba 0.4459 0.0012 1 1>}
  NamedRerouteDeclaration(None=MaterialFunctionCall(Rotation Axis=@n49, Rotation Angle=@n51) {materialFunction=&FixRotateAboutAxisNormals}.New World Space Vertex Normal) {name=New World Normal, nodeColor=<rgba 0.5 0.5 1 1>}
  NamedRerouteUsage()
  NamedRerouteUsage()
  NamedRerouteUsage()
  NamedRerouteUsage()
  n79 = ParticleColor()
  n80 = Multiply(A=@n6) {constB=1.4}
  n81 = Multiply(A=@n80, B=@n80)
  n82 = Multiply(A=@n81, B=@n81)
  Reroute(None=MaterialFunctionCall(Emissive=Multiply(A=@n79.RGB, B=@n82)) {materialFunction=&MF_FX_EmissiveAdjustments}.Result)
  Saturate(None=@n5)
  ScalarParameter() {parameterName=Alpha Transition}
  ScalarParameter() {defaultValue=0.2, parameterName=Noise debug}
  ScalarParameter() {parameterName=Alpha Transition Range}
  ScalarParameter() {parameterName=Expension}
  ScalarParameter() {parameterName=dire}
  TransformPosition() {transformSourceType=TRANSFORMPOSSOURCE_World, transformType=TRANSFORMPOSSOURCE_Instance}
 ]
}
